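# Niagara system for the GothClaws emote that throws out a dust shockwave.
# DustShockwaveParent bursts short-lived particles from a ring, gated by User.Burst probability and slowed by drag.
# DustShockwave spawns sprites from those particles via attribute reader, with per-particle spawn caps scaled by platform quality.
# A scratch swirl module and animated dynamic material parameters shape the dust, while tint and opacity follow time of day with front-end overrides.
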NiagaraSystem /BRCosmetics/Animation/Game/MainPlayer/Emotes/GothClaws/FX/NS_Emote_GothClaws_Dust
dependencies {
 usedDataInterfaces [
  &NiagaraDataInterfaceParticleRead
  &NiagaraDataInterfacePlatformSet
  &NiagaraDataInterfaceCurve
  &NiagaraDataInterfaceColorCurve
 ]
 usedDynamicInputs [&self:ScratchDynamicInput]
 usedModules [
  &SystemState
  &self:SetPeeloutFXArrayIndex
  &self:SystemSpawnScript.SystemScriptSource.SystemScriptGraph.NiagaraNodeAssignment_5.SetVariables_#1
 ]
}
emitters {
 DustShockwave {
  emitterUpdateScript [
   &EmitterState()
   &SpawnParticlesFromOtherEmitter(
    Attribute Reader=DataInterface() {propertyValues {EmitterBinding {bindingMode=Other, emitterName=DustShockwaveParent}}}
    Maximum Age Spawn Threshold=10000
    Spawn Rate=500
    Spawn Rate Per Particle Cap:NiagaraFloat=&MakeCustomFloatFromBool(
     Bool=&IsPlatformSetActive(New Platform Set=DataInterface() {propertyValues {Platforms {qualityLevelMask=-8}}})
     True Float=50
     False Float=&MakeCustomFloatFromBool(
      Bool=&IsPlatformSetActive(New Platform Set=DataInterface() {propertyValues {Platforms {qualityLevelMask=-28}}})
      True Float=40
      False Float=&FN_PickFloatDependOnPlatform(
       TargetPlatform=DataInterface() {propertyValues {Platforms {qualityLevelMask=-30}}}
       TargetPlatformActive=5
       TargetPlatformNoActive=&FN_PickFloatDependOnPlatform(TargetPlatform=DataInterface() {propertyValues {Platforms {qualityLevelMask=-31}}}, TargetPlatformActive=5, TargetPlatformNoActive=0)
      )
     )
    )
    Use Maximum Age=false
    Use Minimum Age=false
   )
  ]
  particleSpawnScript [
   &InitializeParticle(
    Color=<rgba 0.0849 1 0.0015 1>
    Mass=5
    Mass Mode=::NewEnumerator1
    Sprite Rotation Mode=::NewEnumerator2
    UsePositionOffset=false
    Write Material Random=true
   )
   &SampleParticlesFromOtherEmitter(
    Age Sampling:ENiagara_DirectReadSamplingMode=::NewEnumerator0
    Emitter Sampling Mode:ENiagara_EmitterLocSamplingMode=::NewEnumerator0
    Lifetime Sampling:ENiagara_DirectReadSamplingMode=::NewEnumerator0
    Maximum Sampled Particle Age=0.1
    Module.MaxSampledAgeCutoff=true
    Module.MinSampledAgeCutoff=false
    Particle ID Sampling=::NewEnumerator1
    Position Sampling:ENiagara_DirectReadSamplingMode=::NewEnumerator3
    Reject Particle By Sampled Age=true
    Velocity Application:ENiagara_DirectReadApplicationMode=::NewEnumerator0
    Velocity Sampling=::NewEnumerator0
   )
  ]
  particleUpdateScript [
   &ParticleState()
   &SpriteRotationRate(Rotation Rate=&Multiply_Float(
     A=&FloatFromCurve(
      FloatCurve=DataInterface() {propertyValues {Curve {keys [<at 0 1 RCIM_Cubic>, {arriveTangent=-1.0596, tangentMode=RCTM_User, time=1, value=0.2}]}}}
      Scale Curve=200
     )
     B=&RandomRangeFloat(Minimum=-1, Maximum=1)
    ))
   /SetVariables(
    Particles.Color=&MultiplyLinearColor(
     Linear Color=<rgba 0.4281 0.4281 0.4281 0.0675>
     Scale Factor=&MakeVector4FromVectorAndFloat(
      "Vector (RGB)"=&FN_FrontEndVector3Modifier(
       Gameplay=&MakeVectorFromLinearColorRGB(Color=&ColorFromCurve(
         ColorCurve=DataInterface() {propertyValues {
           AlphaCurve {keys [<at 0 1 RCIM_Cubic>]}
           BlueCurve {keys [
             <at 0.083 0.7>
             <at 0.25 1>
             <at 0.3527 1 RCIM_Cubic>
             <at 0.558 1 RCIM_Cubic>
             {arriveTangent=-0.7654, interpMode=RCIM_Cubic, leaveTangent=-0.7654, time=0.875, value=1}
             <at 0.95 0.7>
            ]}
           GreenCurve {keys [
             <at 0.083 0.9356>
             <at 0.25 0.8979>
             {arriveTangent=0.3316, interpMode=RCIM_Cubic, leaveTangent=0.3316, time=0.3527, value=1}
             {arriveTangent=-0.01, interpMode=RCIM_Cubic, leaveTangent=-0.01, time=0.558, value=1}
             {arriveTangent=-0.1568, interpMode=RCIM_Cubic, leaveTangent=-0.1568, time=0.875, value=0.9948}
             <at 0.95 0.9385>
            ]}
           RedCurve {keys [
             <at 0.083 1>
             <at 0.25 0.7>
             {arriveTangent=0.9739, interpMode=RCIM_Cubic, leaveTangent=0.9739, time=0.3527, value=1}
             {arriveTangent=-0.0299, interpMode=RCIM_Cubic, leaveTangent=-0.0299, time=0.558, value=1}
             <at 0.875 0.9844 RCIM_Cubic>
             <at 0.95 1>
            ]}
          }}
         CurveIndex=Linked() {linkedVariable {name=NPC.FortniteNPC.RemappedTimeOfDay_0-1, type=NiagaraFloat}}
        ))
       Front End=<4 4 4>
      )
      "Float (Alpha)"=&FN_FrontEndFloatModifier(Gameplay=&FNTimeOfDayScalar(Night Scale=1.6, Morning Scale=1, Evening Scale=1), Front End=0.6)
     )
    )
    Particles.Scale:NiagaraFloat=&RandomRangeFloat(Minimum=0.4, Maximum=0.55)
   ) {isSetParameters=true}
   &DynamicMaterialParameters(
    Index 0 Param 1=&Add_Float(
     A=&RandomRangeFloat(Minimum=2, Maximum=5)
     B=&FloatFromCurve(
      FloatCurve=DataInterface() {propertyValues {Curve {keys [
          {arriveTangent=2.7533, interpMode=RCIM_Cubic, leaveTangent=1.8958, tangentMode=RCTM_User, time=0, value=1}
          {arriveTangent=0.8292, interpMode=RCIM_Cubic, leaveTangent=0.8292, tangentMode=RCTM_User, time=0.321, value=1.3453}
          {arriveTangent=0.0529, interpMode=RCIM_Cubic, leaveTangent=0.2021, tangentMode=RCTM_User, time=1, value=1.5}
         ]}}}
      CurveIndex=&Add_Float(A=&Multiply_Float(A=$Particles.Initial.MaterialRandom, B=0.5), B=$Particles.NormalizedAge)
     )
    )
    Index 0 Param 2=&FloatFromCurve(FloatCurve=DataInterface() {propertyValues {Curve {keys [
         {arriveTangent=0.0453, interpMode=RCIM_Cubic, tangentMode=RCTM_Break, time=0, value=-1}
         {arriveTangent=0.0604, interpMode=RCIM_Cubic, leaveTangent=0.0604, tangentMode=RCTM_User, time=1, value=-1}
        ]}}})
    Index 0 Param 3=&FloatFromCurve(
     FloatCurve=DataInterface() {propertyValues {Curve {keys [
         {arriveTangent=0.1773, interpMode=RCIM_Cubic, leaveTangent=0.1773, tangentMode=RCTM_User, time=0, value=1}
         {arriveTangent=0.2347, interpMode=RCIM_Cubic, leaveTangent=0.2723, tangentMode=RCTM_Break, time=0.25, value=0}
        ]}}}
     Scale Curve=10
    )
    Index 0 Param 4=&FloatFromCurve(FloatCurve=DataInterface() {propertyValues {Curve {keys [
         {arriveTangent=-2.3762, interpMode=RCIM_Cubic, leaveTangent=-2.3762, tangentMode=RCTM_User, time=0.4, value=1}
         {arriveTangent=-0.5196, interpMode=RCIM_Cubic, leaveTangent=-0.5196, tangentMode=RCTM_User, time=1, value=0}
        ]}}})
    Index 1 Param 1=&FloatFromCurve(
     FloatCurve=DataInterface() {propertyValues {Curve {keys [
         {interpMode=RCIM_Cubic, leaveTangent=4.7088, tangentMode=RCTM_Break, time=0, value=0}
         {arriveTangent=4.2257, interpMode=RCIM_Cubic, leaveTangent=4.2257, tangentMode=RCTM_User, time=0.25, value=1}
        ]}}}
     Scale Curve=&RandomRangeFloat(Minimum=1.5, Maximum=1.5)
    )
    Index 1 Param 2=&FloatFromCurve(
     FloatCurve=DataInterface() {propertyValues {Curve {keys [<at 0 0.2>, {leaveTangent=0.9066, time=1, value=2}]}}}
     Scale Curve=&RandomRangeFloat(Minimum=0.5, Maximum=0.6)
    )
    Index 1 Param 3=&FloatFromCurve(FloatCurve=DataInterface() {propertyValues {Curve {keys [
         {interpMode=RCIM_Cubic, leaveTangent=3.0517, tangentMode=RCTM_Break, time=0, value=0.2}
         {arriveTangent=4.8433, interpMode=RCIM_Cubic, leaveTangent=19.0929, tangentMode=RCTM_User, time=0.25, value=1}
        ]}}})
    Index 1 Param 4=&FloatFromCurve(
     FloatCurve=DataInterface() {propertyValues {Curve {keys [{arriveTangent=-0.3224, interpMode=RCIM_Cubic, leaveTangent=-0.3224, tangentMode=RCTM_User, time=0, value=1}, <at 1 0.9>]}}}
     Scale Curve=&RandomRangeFloat(Minimum=1.25, Maximum=1.75)
    )
    Index 2 Param 1=&FloatFromCurve(FloatCurve=DataInterface() {propertyValues {Curve {
        defaultValue=1
        keys [
         {arriveTangent=9.0725, interpMode=RCIM_Cubic, leaveTangent=9.0725, tangentMode=RCTM_User, time=0, value=0}
         {arriveTangent=-2.5249, interpMode=RCIM_Cubic, leaveTangent=-2.5249, tangentMode=RCTM_User, time=0.1408, value=0.9847}
         {arriveTangent=-0.0405, interpMode=RCIM_Cubic, leaveTangent=-0.0405, tangentMode=RCTM_User, time=1, value=0}
        ]
       }}})
    Index 2 Param 2=&FloatFromCurve(
     FloatCurve=DataInterface() {propertyValues {Curve {keys [
         {arriveTangent=1.133, interpMode=RCIM_Cubic, leaveTangent=1.133, tangentMode=RCTM_User, time=0, value=0.2}
         {arriveTangent=0.5797, interpMode=RCIM_Cubic, leaveTangent=0.5797, tangentMode=RCTM_User, time=1, value=1}
        ]}}}
     Scale Curve=&RandomRangeFloat(Minimum=0.2, Maximum=0.5)
    )
    Index 2 Param 3=&FloatFromCurve(FloatCurve=DataInterface() {propertyValues {Curve {keys [
         {interpMode=RCIM_Cubic, leaveTangent=0.7449, tangentMode=RCTM_Break, time=0, value=1}
         {arriveTangent=0.2809, interpMode=RCIM_Cubic, leaveTangent=11.1233, tangentMode=RCTM_Break, time=1, value=1.5}
        ]}}})
    Index 2 Param 4=&FloatFromCurve(
     FloatCurve=DataInterface() {propertyValues {Curve {keys [
         {arriveTangent=-0.4633, interpMode=RCIM_Cubic, leaveTangent=-0.4633, tangentMode=RCTM_User, time=0, value=1}
         {arriveTangent=-0.1168, interpMode=RCIM_Cubic, leaveTangent=-1.2956, tangentMode=RCTM_Break, time=1, value=0.75}
        ]}}}
     Scale Curve=2
    )
    Module.Param0WriteEnabled=true
    Module.Param0WriteEnabled1=true
    Module.Param0WriteEnabled2=true
    Module.Param1WriteEnabled=false
    Module.Param1WriteEnabled1=true
    Module.Param1WriteEnabled2=true
    Module.Param2WriteEnabled=false
    Module.Param2WriteEnabled1=true
    Module.Param2WriteEnabled2=true
    Module.Param3WriteEnabled=true
    Module.Param3WriteEnabled1=true
    Module.Param3WriteEnabled2=true
    Write Parameter Index 1=true
    Write Parameter Index 2=true
   )
   &SolveForcesAndVelocity(Clamp Velocity=false, Limit Acceleration=false)
   /Niagara/Modules/AttributeReader/UpdateParticlesFromOtherEmitter(
    Attribute Reader=DataInterface() {propertyValues {EmitterBinding {bindingMode=Other, emitterName=DustShockwaveParent}}}
    Velocity Sampling:ENiagara_DirectReadSamplingMode=::NewEnumerator0
   )
   &self:Swirl_01(
    OpacityFadePower=2.5
    RotInnterSpeedup=1
    RotSpeed=-0.5
    Scale:NiagaraFloat=&RandomRangeFloat(Minimum=30, Maximum=60)
    ScaleLength=1
    ScaleUpByNormAge=1
    ScaleVar=0.5
   ) {module=Swirl_01}
  ]
  rendererClasses [&NiagaraSpriteRendererProperties]
  renderers [{rendererClass=&NiagaraSpriteRendererProperties}]
 }
 DustShockwaveParent {
  emitterUpdateScript [
   &EmitterState(Inactive Response=::NewEnumerator1, Life Cycle Mode=::NewEnumerator1, Loop Delay=0.5, Loop Duration=0.1, UseLoopDelay=false)
   &SpawnBurst_Instantaneous(Loop Count Limit=5, Spawn Count=20, Spawn Probability=$User.Burst, Use Spawn Probability=true, UseLoopCountLimit=false)
  ]
  particleSpawnScript [
   &InitializeParticle(
    Color=<rgba 1 0.0572 0 1>
    Lifetime Max=1
    Lifetime Min=0.5
    Lifetime Mode=::NewEnumerator1
    Mass=5
    Mass Mode=::NewEnumerator1
    Sprite Rotation Mode=::NewEnumerator2
    Sprite Size Mode=::NewEnumerator1
    Uniform Sprite Size Max=12
    Uniform Sprite Size Min=6
    UsePositionOffset=false
    Write Material Random=true
   )
   &ShapeLocation(
    Additional Quaternion Rotation=true
    Additional Yaw / Pitch / Roll=<0.35 0 0>
    MaskSpawnGroup=false
    Offset=<0 -1 0>
    Offset Mode=::NewEnumerator0
    Override Seed=false
    Ring Radius=50
    Rotation Mode=::NewEnumerator3
    Shape Primitive=::NewEnumerator4
    U Distribution=0.8
    UseOwnerScale=true
   )
   &ApplyInitialForces()
   &AddVelocity(
    Module.Use Falloff Distance=false
    Origin Offset=<0 -5 0>
    Override Seed=false
    Velocity Mode=::NewEnumerator1
    Velocity Speed=&RandomRangeFloat(Minimum=650, Maximum=850)
   )
  ]
  particleUpdateScript [
   &ParticleState()
   &Drag(Drag=&RandomRangeFloat(Minimum=2, Maximum=4), Use Linear Drag=true, Use Rotational Drag=true)
   &SolveForcesAndVelocity(Clamp Velocity=false, Limit Acceleration=false)
   &KillParticles(Kill Particles=HlslExpression() {hlslExpression="Particles.Age > 0.5"})
  ]
  rendererClasses [&NiagaraSpriteRendererProperties]
  renderers [{rendererClass=&NiagaraSpriteRendererProperties}]
 }
}
summary {
 systemName=NS_Emote_GothClaws_Dust
 emitters [
  {emitterName=DustShockwaveParent, rendererClasses [&NiagaraSpriteRendererProperties]}
  {emitterName=DustShockwave, rendererClasses [&NiagaraSpriteRendererProperties]}
 ]
 userVariables [
  User.bIsFrontend {type=NiagaraBool}
  User.bIsFrontendPreview {type=NiagaraBool}
  User.Burst {type=NiagaraFloat, defaultValue=1, description=""}
  User.FEScale {type=NiagaraFloat, defaultValue=0.6, description=""}
  User.IsSpawnDust {type=NiagaraBool, defaultValue=true}
  User.SlidingColorArray {
   type=NiagaraDataInterfaceArrayColor
   description=""
   defaultValue {
    struct=NiagaraExt_VariableValue_DataInterface
    value {
     dataInterface=&self:NiagaraDataInterfaceArrayColor_0
     dataInterfaceClass=&NiagaraDataInterfaceArrayColor
    }
   }
  }
  User.SpawnRateMult {type=NiagaraFloat, defaultValue=1, description=""}
  User.SurfaceTypeIndex {type=NiagaraInt32}
 ]
}
