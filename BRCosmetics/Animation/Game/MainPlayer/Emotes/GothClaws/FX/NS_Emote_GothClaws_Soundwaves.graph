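# Niagara system for the GothClaws emote that emits soundwave effects from speaker points sampled on a skeletal mesh.
# Rings spawns travelling mesh waves with jittered red/blue offsets and refraction parameters, Blinking_Streaks adds streaks, and Speaker_Ring_Streaks_Glows holds pulsing rings, glows and lights on the speakers.
# User.Burst and User.SpawnRateMult drive rate or intensity, with time-of-day, front-end and camera-angle adjustments.
# Each emitter merges several layers through spawn groups and VisibilityTag switches.
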NiagaraSystem /BRCosmetics/Animation/Game/MainPlayer/Emotes/GothClaws/FX/NS_Emote_GothClaws_Soundwaves
dependencies {
 usedDataInterfaces [
  &NiagaraDataInterfaceCurve
  &NiagaraDataInterfaceColorCurve
  &NiagaraDataInterfaceCamera
  /Script/Niagara.NiagaraDataInterfaceArrayDistributionInt
  &NiagaraDataInterfaceArrayFloat
  &NiagaraDataInterfaceVector2DCurve
  &NiagaraDataInterfaceSkeletalMesh
 ]
 usedDynamicInputs [&MakeFloatFromVector2d, &FNTimeOfDayColor]
 usedModules [
  &SystemState
  &self:SystemSpawnScript.SystemScriptSource.SystemScriptGraph.NiagaraNodeAssignment_0.SetVariables_#1
  &self:SystemSpawnScript.SystemScriptSource.SystemScriptGraph.NiagaraNodeAssignment_1.SetVariables_#2
  &self:SystemSpawnScript.SystemScriptSource.SystemScriptGraph.NiagaraNodeAssignment_2.SetVariables_#3
  &self:SystemSpawnScript.SystemScriptSource.SystemScriptGraph.NiagaraNodeAssignment_3.SetVariables_#4
  &self:SystemSpawnScript.SystemScriptSource.SystemScriptGraph.NiagaraNodeAssignment_4.SetVariables_#5
 ]
}
emitters {
 Blinking_Streaks {
  emitterUpdateScript [
   &EmitterState(
    Inactive Response=::NewEnumerator1
    Life Cycle Mode=::NewEnumerator1
    Loop Behavior=::NewEnumerator1
    Loop Delay=0.5
    Loop Duration Mode=::NewEnumerator1
    UseLoopDelay=true
   )
   &SpawnRate(
    SpawnRate=&Multiply_Float(
     A=7
     B=&Multiply_Float(
      A=$User.Burst
      B=&RemapRange(Input Value To Remap=$User.SpawnRateMult, Input Min=1, Input Max=2, Output Min=4, Output Max=2)
     )
    )
    Use Spawn Probability=false
   )
   &SpawnRate(
    SpawnGroup=1
    SpawnRate=&Multiply_Float(
     A=30
     B=&Multiply_Float(
      A=$User.Burst
      B=&RemapRange(Input Value To Remap=$User.SpawnRateMult, Input Min=1, Input Max=2, Output Min=4, Output Max=2)
     )
    )
    Use Spawn Probability=false
   ) {module=SpawnRate001}
  ]
  particleSpawnScript [
   /SetVariables(Particles.VisibilityTag=&SpawnGroup()) {isSetParameters=true}
   &PartitionParticles()
   &PartitionParticles(Partition Mode:ENiagara_PartitionMode_v2=::NewEnumerator5) {module=PartitionParticles001}
   &InitializeParticle(
    Color=&MergeHelper_ColorSwitch(0=<rgba 1.5 1.5 1.5 0.1>, 1=<rgba 5 5 5 0.1>, 2=?, 3=?, Index=$Particles.VisibilityTag, Num=::NewEnumerator0)
    Lifetime=&MergeHelper_FloatSwitch(
     0=&Multiply_Float(A=0.4, B=0.5)
     1=&Multiply_Float(A=0.25, B=0.45)
     2=?
     3=?
     Index=$Particles.VisibilityTag
     Num=::NewEnumerator0
    )
    Mesh Scale=<4 4 0.5>
    Mesh Scale Mode=::NewEnumerator4
    Sprite Size=&Multiply_Vector2D_ByFloat(
     Float=&SelectFloatFromArray(
      Direct Array Index=/Niagara/DynamicInputs/Arrays/SelectIntFromWeightedDistributionArray(
       Array Sampling Mode:ENiagara_LifetimeMode=::NewEnumerator1
       Direct Array Index:NiagaraInt32=?
       Direct Array Mode:ENiagara_WrapClamp=::NewEnumerator0
       Evaluation Type:ENiagaraRandomnessEvaluation=::NewEnumerator0
       Recalculate Random Each Loop:NiagaraBool=false
       Weighted Distribution Int Array:NiagaraDataInterfaceArrayDistributionInt=DataInterface() {propertyValues {ArrayData [{value=0, weight=0.15}, {value=1, weight=0.35}, {value=2, weight=0.5}]}}
       Randomness Mode:ENiagaraRandomnessMode=::NewEnumerator0
       Random Seed:NiagaraInt32=0
       Fixed Random Seed:NiagaraBool=false
      )
      Float Selection Array=DataInterface() {propertyValues {FloatData [0.5,0.8,1]}}
     )
     Vector2D=&RandomRangeVector2D(Minimum=<5 30>, Maximum=<15 55>)
    )
    Sprite Size Mode=::NewEnumerator4
    UsePositionOffset=false
    Write Material Random=true
   )
   &SkeletalMeshLocation(
    Bone / Socket Index=&MergeHelper_IntSwitch(
     0=$Particles.PartitionParticles.Partition
     1=$Particles.PartitionParticles001.Partition
     2=?
     3=?
     4=?
     5=?
     Index=$Particles.VisibilityTag
     Num=::NewEnumerator0
    )
    Bone Sampling Mode=::NewEnumerator4
    MaskSpawnGroup=false
    Sampled Position Offset=<0 5 0>
    Skeletal Mesh=$System.Skeletal Mesh Reader_Speakers
   ) {module=SkeletalMeshLocation001}
   &DynamicMaterialParameters(
    Index 0 Param 2=1
    Index 0 Param 3=1
    Index 0 Param 4=1
    Module.Param0WriteEnabled=false
    Module.Param1WriteEnabled=true
    Module.Param2WriteEnabled=true
    Module.Param3WriteEnabled=true
   )
   &InitialMeshOrientation(
    Mesh Orientation Mode=::NewEnumerator6
    Orientation Quaternion:Quat4f=<-1 0 0 1>
    Rotation=&RandomRangeVector(Minimum=<0 0 0>, Maximum=<0 0 1>)
    Use Rotation Vector=true
   )
   &ShapeLocation(
    MaskSpawnGroup=false
    Override Seed=false
    Ring Radius=&MaskFloat(Mask Source=0, Mask Parameter=?, Mask=1, Masked Float=14.5, Passthrough Float=0)
    Rotation Mode=::NewEnumerator2
    Shape Primitive=::NewEnumerator4
    U Distribution=0.5
    UseOwnerScale=true
    Yaw / Pitch / Roll=&MergeHelper_VectorSwitch(0=<0 -0.25 0.25>, 1=<0 0.25 0.25>, 2=?, 3=?, Index=$Particles.PartitionParticles001.Partition, Num=::NewEnumerator0)
   )
   &AddVelocity(
    Module.Use Falloff Distance=false
    Origin Offset=&MergeHelper_VectorSwitch(
     0=<0 0 0>
     1=&AddVectorToPosition(Position=$Particles.SkeletalMeshLocation001.SampledPosition, Vector=<0 -10 0>)
     2=?
     3=?
     Index=$Particles.VisibilityTag
     Num=::NewEnumerator0
    )
    Override Seed=false
    Velocity Mode=::NewEnumerator1
    Velocity Speed=&MergeHelper_FloatSwitch(
     0=0
     1=&RandomRangeFloat(Minimum=1, Maximum=1)
     2=?
     3=?
     Index=&SpawnGroup()
     Num=::NewEnumerator0
    )
   ) {module=AddVelocity001}
  ]
  particleUpdateScript [
   &ParticleState()
   &ScaleMeshSize(Scale Factor=&VectorFromFloat(Value=&FloatFromCurve(
      FloatCurve=DataInterface() {propertyValues {Curve {
         defaultValue=0
         keys [
          {arriveTangent=2.0673, interpMode=RCIM_Cubic, leaveTangent=2.0673, tangentMode=RCTM_User, time=0, value=0}
          {interpMode=RCIM_Cubic, tangentMode=RCTM_User, time=1, value=1}
         ]
        }}}
      Scale Curve=&RandomRangeFloat(Minimum=1.15, Maximum=1.3)
     )))
   &ScaleSpriteSize(
    Scale Factor=&Vector2DFromCurve(
     Vector2Curve=DataInterface() {propertyValues {
       XCurve {keys [{arriveTangent=0.7701, interpMode=RCIM_Cubic, leaveTangent=0.7701, tangentMode=RCTM_User, time=0, value=0.6}, <at 1 1 RCIM_Cubic>]}
       YCurve {keys [
         {arriveTangent=1.7762, interpMode=RCIM_Cubic, leaveTangent=1.7762, tangentMode=RCTM_User, time=0, value=0.2}
         {arriveTangent=0.5255, leaveTangent=2.2588, tangentMode=RCTM_User, time=1, value=1.5}
        ]}
      }}
     ScaleCurve=&MakeVector2D(
      X=1
      Y=&Multiply_Float(
       A=&FN_FrontEndFloatModifier(Gameplay=1, Front End=0.8)
       B=&RemapRange(Input Value To Remap=$User.SpawnRateMult, Input Min=1, Input Max=2, Output Min=2.5, Output Max=3.5)
      )
     )
    )
    Scale Sprite Size Mode=::NewEnumerator2
   )
   &ScaleColor(
    Scale Alpha=&Multiply_Float(
     A=&FNTimeOfDayScalar(Night Scale=0.85, Morning Scale=0.9, Evening Scale=0.8)
     B=&MergeHelper_FloatSwitch(
      0=&FloatFromCurve(
       FloatCurve=DataInterface() {propertyValues {Curve {
          defaultValue=0
          keys [
           {interpMode=RCIM_Cubic, tangentMode=RCTM_User, time=0, value=0}
           {interpMode=RCIM_Cubic, tangentMode=RCTM_User, time=0.5, value=1}
           {tangentMode=RCTM_User, time=1, value=0}
          ]
         }}}
       Scale Curve=&FN_FrontEndFloatModifier(
        Gameplay=&FNTimeOfDayScalar(Night Scale=0.8, Morning Scale=0.8, Evening Scale=0.8)
        Front End=1.1
       )
      )
      1=&FloatFromCurve(
       FloatCurve=DataInterface() {propertyValues {Curve {keys [
           {arriveTangent=9.6668, interpMode=RCIM_Cubic, leaveTangent=9.6668, tangentMode=RCTM_User, time=0, value=0}
           <at 0.0951 1 RCIM_Cubic>
           {arriveTangent=-1.7835, interpMode=RCIM_Cubic, leaveTangent=-1.7835, tangentMode=RCTM_User, time=1, value=0}
          ]}}}
       Scale Curve=&Multiply_Float(
        A=&RandomRangeFloat(Minimum=0.25, Maximum=0.4)
        B=&RemapRange(
         Input Value To Remap=&Abs_Float(Float:NiagaraFloat=&DotProduct(Vector A=&ReturnCameraVector(Camera Vectors=::NewEnumerator1, Coordinate Space=::Simulation), Vector B=<0 1 0>))
         Input Min=0.1
         Input Max=1
         Output Min=0.6
         Output Max=1
        )
       )
      )
      2=?
      3=?
      Index=$Particles.VisibilityTag
      Num=::NewEnumerator0
     )
    )
    Scale RGB=&MergeHelper_VectorSwitch(
     0=<1 1 1>
     1=&MakeVectorFromLinearColorRGB(Color=&ColorFromCurve(
       ColorCurve=DataInterface() {propertyValues {
         AlphaCurve {keys [<at 0 1 RCIM_Cubic>]}
         BlueCurve {keys [
           <at 0.083 1>
           <at 0.25 0.8469>
           {arriveTangent=0.4971, interpMode=RCIM_Cubic, leaveTangent=0.4971, time=0.3527, value=1}
           {arriveTangent=-0.2872, interpMode=RCIM_Cubic, leaveTangent=-0.2872, time=0.558, value=1}
           <at 0.875 0.85 RCIM_Cubic>
           <at 0.95 1>
          ]}
         GreenCurve {keys [
           <at 0.083 0.6667>
           <at 0.25 0.8469>
           {arriveTangent=0.4971, interpMode=RCIM_Cubic, leaveTangent=0.4971, time=0.3527, value=1}
           {arriveTangent=-0.2872, interpMode=RCIM_Cubic, leaveTangent=-0.2872, time=0.558, value=1}
           {arriveTangent=-0.8504, interpMode=RCIM_Cubic, leaveTangent=-0.8504, time=0.875, value=0.85}
           <at 0.95 0.6667>
          ]}
         RedCurve {keys [
           <at 0.083 0.5>
           <at 0.25 1>
           <at 0.3527 1 RCIM_Cubic>
           <at 0.558 1 RCIM_Cubic>
           {arriveTangent=-1.2756, interpMode=RCIM_Cubic, leaveTangent=-1.2756, time=0.875, value=1}
           <at 0.95 0.5>
          ]}
        }}
       CurveIndex=Linked() {linkedVariable {name=NPC.FortniteNPC.RemappedTimeOfDay_0-1, type=NiagaraFloat}}
      ))
     2=?
     3=?
     Index=$Particles.VisibilityTag
     Num=::NewEnumerator0
    )
    ScaleA=true
    ScaleRGB=true
   )
   &Drag(
    Drag=&MergeHelper_FloatSwitch(
     0=0
     1=&RandomRangeFloat(Minimum=1, Maximum=2)
     2=?
     3=?
     Index=$Particles.VisibilityTag
     Num=::NewEnumerator0
    )
    Rotational Drag:NiagaraFloat=&MergeHelper_FloatSwitch(0=0, 1=1, 2=?, 3=?, Index=$Particles.VisibilityTag, Num=::NewEnumerator0)
    Use Linear Drag=true
    Use Rotational Drag=true
   )
   &MeshRotationForce(Rotation=&MakeVector(
     X=&FloatFromCurve(FloatCurve=DataInterface() {propertyValues {Curve {keys [<at 0 1>,<at 1 0>]}}})
     Y=0
     Z=0
    ))
   &SolveForcesAndVelocity(Clamp Velocity=false, Limit Acceleration=false)
   /SetVariables(Particles.DynamicMaterialParameter0_Spikes:Vector4f=<0 0 0 0>, Particles.DynamicMaterialParameter1_Spikes:Vector4f=<0.2 0.2 0 0>) {isSetParameters=true}
  ]
  rendererClasses [&NiagaraMeshRendererProperties, &NiagaraSpriteRendererProperties]
  renderers [
   {rendererClass=&NiagaraMeshRendererProperties, rendererIndex=0}
   {rendererClass=&NiagaraSpriteRendererProperties, rendererIndex=1}
  ]
 }
 Rings {
  emitterUpdateScript [
   &EmitterState(
    Inactive Response=::NewEnumerator1
    Life Cycle Mode=::NewEnumerator1
    Loop Behavior=::NewEnumerator1
    Loop Delay=0.35
    Loop Duration Mode=::NewEnumerator1
    UseLoopDelay=true
   )
   &SpawnRate(
    SpawnRate=&Multiply_Float(
     A=15
     B=&Multiply_Float(
      A=$User.Burst
      B=&RemapRange(Input Value To Remap=$User.SpawnRateMult, Input Min=1, Input Max=2, Output Min=1, Output Max=0.75)
     )
    )
    Use Spawn Probability=false
   ) {module=SpawnRate001}
   &SpawnRate(
    SpawnGroup=1
    SpawnRate=&Multiply_Float(
     A=15
     B=&Multiply_Float(
      A=$User.Burst
      B=&RemapRange(Input Value To Remap=$User.SpawnRateMult, Input Min=1, Input Max=2, Output Min=1, Output Max=0.75)
     )
    )
    Use Spawn Probability=false
   )
   &SpawnRate(
    SpawnGroup=2
    SpawnRate=&Multiply_Float(
     A=22
     B=&Multiply_Float(
      A=$User.Burst
      B=&RemapRange(Input Value To Remap=$User.SpawnRateMult, Input Min=1, Input Max=2, Output Min=0.5, Output Max=0.3)
     )
    )
    Use Spawn Probability=false
   ) {module=SpawnRate002}
   &SpawnRate(
    SpawnGroup=3
    SpawnRate=&Multiply_Float(
     A=22
     B=&Multiply_Float(
      A=$User.Burst
      B=&RemapRange(Input Value To Remap=$User.SpawnRateMult, Input Min=1, Input Max=2, Output Min=0.5, Output Max=0.3)
     )
    )
    Use Spawn Probability=false
   ) {module=SpawnRate003}
   &SpawnRate(
    SpawnGroup=4
    SpawnRate=&Multiply_Float(
     A=5
     B=&Multiply_Float(
      A=$User.Burst
      B=&RemapRange(Input Value To Remap=$User.SpawnRateMult, Input Min=1, Input Max=2, Output Min=1, Output Max=1)
     )
    )
    Use Spawn Probability=false
   ) {module=SpawnRate004}
   &SpawnRate(
    SpawnGroup=5
    SpawnRate=&Multiply_Float(
     A=10
     B=&Multiply_Float(
      A=$User.Burst
      B=&RemapRange(Input Value To Remap=$User.SpawnRateMult, Input Min=1, Input Max=2, Output Min=1, Output Max=1)
     )
    )
    Use Spawn Probability=false
   ) {module=SpawnRate005}
  ]
  particleSpawnScript [
   /SetVariables(Particles.VisibilityTag=&MergeHelper_IntSwitch(0=0, 1=0, 2=1, 3=1, 4=2, 5=2, Index=&SpawnGroup(), Num=::NewEnumerator4)) {isSetParameters=true}
   &InitializeParticle(
    Color=&MergeHelper_ColorSwitch(0=<rgba 0.4 1 0.4 1>, 1=<rgba 0.3 1 0.3 1>, 2=<rgba 4.5013 5 4.375 0.15>, 3=?, Index=$Particles.VisibilityTag, Num=::NewEnumerator1)
    Lifetime=&MergeHelper_FloatSwitch(
     0=0.2
     1=&FN_FrontEndFloatModifier(Gameplay=0.25, Front End=0.2)
     2=0.4
     3=?
     Index=$Particles.VisibilityTag
     Num=::NewEnumerator1
    )
    Mesh Scale=&MergeHelper_VectorSwitch(
     0=&VectorFromFloat(Value=0.1)
     1=&VectorFromFloat(Value=&Multiply_Float(
       A=0.1
       B=&RemapRange(Input Value To Remap=$User.SpawnRateMult, Clamp Results=false, Input Min=1, Input Max=2, Output Min=0.95, Output Max=1.15)
      ))
     2=&Multiply_VectorByFloat(
      Vector=<0.6 0.6 -0.2>
      Float=&RemapRange(Input Value To Remap=$User.SpawnRateMult, Clamp Results=false, Input Min=1, Input Max=2, Output Min=0.95, Output Max=1.15)
     )
     3=?
     Index=$Particles.VisibilityTag
     Num=::NewEnumerator1
    )
    Mesh Scale Mode=::NewEnumerator4
    UsePositionOffset=false
    Write Material Random=true
   )
   &SkeletalMeshLocation(
    Bone / Socket Index=&MergeHelper_IntSwitch(0=0, 1=1, 2=0, 3=1, 4=0, 5=1, Index=&SpawnGroup(), Num=::NewEnumerator4)
    Bone Sampling Mode=::NewEnumerator4
    MaskSpawnGroup=false
    Sampled Position Offset=<0 4 0>
    Skeletal Mesh=$System.Skeletal Mesh Reader_Speakers
   )
   &Color(Color=&MultiplyLinearColor(
     Linear Color=$Particles.Color
     Scale Factor=&MakeVector4(
      X=1
      Y=1
      Z=1
      W=&MergeHelper_FloatSwitch(
       0=&Multiply_Float(
        A=0.08
        B=&RemapRange(Input Value To Remap=$User.SpawnRateMult, Input Min=1, Input Max=2, Output Min=1, Output Max=0.5)
       )
       1=&Multiply_Float(
        A=0.22
        B=&RemapRange(Input Value To Remap=$User.SpawnRateMult, Input Min=1, Input Max=2, Output Min=1, Output Max=0.5)
       )
       2=?
       3=?
       Index=$Particles.VisibilityTag
       Num=::NewEnumerator0
      )
     )
    ))
   &AddVelocity(
    Velocity=&RandomRangeVector(
     Minimum=&MergeHelper_VectorSwitch(0=<0 25 0>, 1=<0 40 0>, 2=<0 40 0>, 3=?, Index=$Particles.VisibilityTag, Num=::NewEnumerator1)
     Maximum=&MergeHelper_VectorSwitch(0=<0 60 0>, 1=<0 60 0>, 2=<0 60 0>, 3=?, Index=$Particles.VisibilityTag, Num=::NewEnumerator1)
    )
    Velocity Speed Scale=&MergeHelper_FloatSwitch(
     0=&Multiply_Float(
      A=&RandomRangeFloat(Minimum=2, Maximum=2.25)
      B=&RemapRange(Input Value To Remap=$User.SpawnRateMult, Input Min=1, Input Max=2, Output Min=2.5, Output Max=5)
     )
     1=&Multiply_Float(
      A=&RandomRangeFloat(Minimum=1.5, Maximum=1.75)
      B=&RemapRange(Input Value To Remap=$User.SpawnRateMult, Input Min=1, Input Max=2, Output Min=5, Output Max=9.5)
     )
     2=&Multiply_Float(
      A=1.1
      B=&RemapRange(Input Value To Remap=$User.SpawnRateMult, Input Min=1, Input Max=2, Output Min=2, Output Max=6)
     )
     3=?
     Index=$Particles.VisibilityTag
     Num=::NewEnumerator1
    )
   )
   &InitialMeshOrientation(
    Mesh Orientation Mode=::NewEnumerator6
    Orientation Quaternion:Quat4f=/Game/Effects/Niagara/Modules/MergeEmitters/MergeHelper_QuatSwitch(
     0:Quat4f=<1 0 0 1>
     1:Quat4f=<0 0 0 1>
     2:Quat4f=<0 0 0 1>
     3:Quat4f=?
     4:Quat4f=?
     5:Quat4f=?
     6:Quat4f=?
     7:Quat4f=?
     8:Quat4f=?
     9:Quat4f=?
     Index:NiagaraInt32=$Particles.VisibilityTag
     Num:MergeHelper_SwitcherEnum=::NewEnumerator1
    )
    Rotation=&MergeHelper_VectorSwitch(0=<0 0 0>, 1=<0.25 0 0>, 2=<0.25 0 0>, 3=?, Index=$Particles.VisibilityTag, Num=::NewEnumerator1)
    Use Rotation Vector=true
   )
   /SetVariables(Particles.DynamicMaterialParameter1_Big:Vector4f=&MakeVector4(
     X=&RandomRangeFloat(Minimum=0, Maximum=5)
     Y=&RandomRangeFloat(Minimum=-0.02, Maximum=-0.1)
     Z=&RandomRangeFloat(Minimum=0, Maximum=1)
     W=&RandomRangeFloat(Minimum=0, Maximum=1)
    )) {isSetParameters=true}
  ]
  particleUpdateScript [
   &ParticleState()
   &ScaleMeshSize(Scale Factor=&MergeHelper_VectorSwitch(
     0=&VectorFromFloat(Value=&FloatFromCurve(
       FloatCurve=DataInterface() {propertyValues {Curve {keys [{arriveTangent=1.0728, interpMode=RCIM_Cubic, leaveTangent=1.0728, tangentMode=RCTM_User, time=0, value=0.5}, <at 1 1.75 RCIM_Cubic>]}}}
       Scale Curve=&FN_FrontEndFloatModifier(Gameplay=1.8, Front End=1.3)
      ))
     1=&VectorFromFloat(Value=&FloatFromCurve(
       FloatCurve=DataInterface() {propertyValues {Curve {keys [{arriveTangent=1.0728, interpMode=RCIM_Cubic, leaveTangent=1.0728, tangentMode=RCTM_User, time=0, value=0.5}, <at 1 2 RCIM_Cubic>]}}}
       Scale Curve=&FN_FrontEndFloatModifier(Gameplay=3.5, Front End=2.6)
      ))
     2=&VectorFromFloat(Value=&FloatFromCurve(
       FloatCurve=DataInterface() {propertyValues {Curve {keys [{arriveTangent=1.0728, interpMode=RCIM_Cubic, leaveTangent=1.0728, tangentMode=RCTM_User, time=0, value=0.5}, <at 1 1.5 RCIM_Cubic>]}}}
       Scale Curve=3.2
      ))
     3=?
     Index=$Particles.VisibilityTag
     Num=::NewEnumerator1
    ))
   &ScaleColor(
    Scale Alpha=&Multiply_Float(
     A=&FN_FrontEndFloatModifier(
      Gameplay=&FNTimeOfDayScalar(Night Scale=0.5, Morning Scale=0.8, Evening Scale=0.5)
      Front End=1.15
     )
     B=&MergeHelper_FloatSwitch(
      0=&FloatFromCurve(
       FloatCurve=DataInterface() {propertyValues {Curve {keys [
           {arriveTangent=51.3864, interpMode=RCIM_Cubic, leaveTangent=51.3864, tangentMode=RCTM_User, time=0, value=1}
           {arriveTangent=-0.4511, interpMode=RCIM_Cubic, leaveTangent=-0.4511, tangentMode=RCTM_User, time=0.0538, value=3.5}
           {arriveTangent=-4.1942, interpMode=RCIM_Cubic, leaveTangent=-4.1942, tangentMode=RCTM_User, time=0.9, value=0}
          ]}}}
       Scale Curve=&Multiply_Float(
        A=&FN_FrontEndFloatModifier(Gameplay=1, Front End=2)
        B=&RemapRange(
         Input Value To Remap=&Abs_Float(Float:NiagaraFloat=&DotProduct(Vector A=&ReturnCameraVector(Camera Vectors=::NewEnumerator1, Coordinate Space=::Simulation), Vector B=<0 1 0>))
         Input Min=0.1
         Input Max=1
         Output Min=0.5
         Output Max=1
        )
       )
      )
      1=&FloatFromCurve(
       FloatCurve=DataInterface() {propertyValues {Curve {keys [
           {arriveTangent=51.3864, interpMode=RCIM_Cubic, leaveTangent=51.3864, tangentMode=RCTM_User, time=0, value=1}
           {arriveTangent=-0.4511, interpMode=RCIM_Cubic, leaveTangent=-0.4511, tangentMode=RCTM_User, time=0.0538, value=2}
           {arriveTangent=-4.379, interpMode=RCIM_Cubic, leaveTangent=-4.379, tangentMode=RCTM_User, time=0.975, value=0}
          ]}}}
       Scale Curve=&RemapRange(
        Input Value To Remap=&Abs_Float(Float:NiagaraFloat=&DotProduct(Vector A=&ReturnCameraVector(Camera Vectors=::NewEnumerator1, Coordinate Space=::Simulation), Vector B=<0 1 0>))
        Input Min=0.1
        Input Max=1
        Output Min=0.35
        Output Max=0.7
       )
      )
      2=1
      3=?
      Index=$Particles.VisibilityTag
      Num=::NewEnumerator1
     )
    )
    Scale RGB=&MakeVectorFromLinearColorRGB(Color=&ColorFromCurve(
      ColorCurve=DataInterface() {propertyValues {
        AlphaCurve {keys [<at 0 1 RCIM_Cubic>]}
        BlueCurve {keys [
          <at 0.1 2.5>
          <at 0.25 0.8469>
          {arriveTangent=0.4971, interpMode=RCIM_Cubic, leaveTangent=0.4971, time=0.3527, value=1}
          {arriveTangent=-0.2872, interpMode=RCIM_Cubic, leaveTangent=-0.2872, time=0.558, value=1}
          {arriveTangent=2.5512, interpMode=RCIM_Cubic, leaveTangent=2.5512, time=0.875, value=0.85}
          <at 0.95 2>
         ]}
        GreenCurve {keys [
          <at 0.1 0.7913>
          <at 0.25 0.8469>
          {arriveTangent=0.4971, interpMode=RCIM_Cubic, leaveTangent=0.4971, time=0.3527, value=1}
          {arriveTangent=-0.2872, interpMode=RCIM_Cubic, leaveTangent=-0.2872, time=0.558, value=1}
          {arriveTangent=-0.5324, interpMode=RCIM_Cubic, leaveTangent=-0.5324, time=0.875, value=0.85}
          <at 0.95 0.7913>
         ]}
        RedCurve {keys [
          <at 0.1 0>
          <at 0.25 1>
          <at 0.3527 1 RCIM_Cubic>
          <at 0.558 1 RCIM_Cubic>
          {arriveTangent=-2.5512, interpMode=RCIM_Cubic, leaveTangent=-2.5512, time=0.875, value=1}
          <at 0.95 0>
         ]}
       }}
      CurveIndex=Linked() {linkedVariable {name=NPC.FortniteNPC.RemappedTimeOfDay_0-1, type=NiagaraFloat}}
     ))
    ScaleA=true
    ScaleRGB=true
   )
   &DynamicMaterialParameters(
    Index 0 Param 1=&Multiply_Float(A=&MakeFloatFromVector(Vector=$Particles.Scale, Channel=::NewEnumerator0), B=1.2)
    Index 0 Param 2=50
    Index 0 Param 3=0.25
    Index 0 Param 4=50
    Index 1 Param 1=15
    Index 1 Param 2=-0.07
    Index 1 Param 3=0.1
    Index 1 Param 4=$Particles.Lifetime
    Index 2 Param 1=2
    Module.Index0Param0Write=true
    Module.Index0Param1Write=true
    Module.Index0Param2Write=true
    Module.Index0Param3Write=true
    Module.Index1Param0Write=true
    Module.Index1Param1Write=true
    Module.Index1Param2Write=true
    Module.Index1Param3Write=true
    Module.Index2Param0Write=true
    Module.Index2Param1Write=true
    Module.Index2Param2Write=true
    Module.Index2Param3Write=true
    Module.Index3Param0Write=true
    Module.Index3Param1Write=true
    Module.Index3Param2Write=true
    Module.Index3Param3Write=true
    Write Parameter Index 1=true
    Write Parameter Index 2=true
    Write Parameter Index 3=true
   )
   &Drag(
    Drag=&MergeHelper_FloatSwitch(
     0=0
     1=&RandomRangeFloat(Minimum=0.5, Maximum=1.25)
     2=&RandomRangeFloat(Minimum=0.5, Maximum=1.25)
     3=?
     Index=$Particles.VisibilityTag
     Num=::NewEnumerator1
    )
    Use Linear Drag=true
    Use Rotational Drag=false
   )
   &SolveForcesAndVelocity(Clamp Velocity=false, Limit Acceleration=false)
   /SetVariables(
    Particles.ColorBlueSmall:LinearColor=&MakeLinearColor(
     R:NiagaraFloat=0.2
     G:NiagaraFloat=0.2
     B:NiagaraFloat=1
     A:NiagaraFloat=&MakeFloatFromLinearColor(LinearColor=$Particles.Color, Channel:ENiagaraFloatFromLinearColorOptions=::NewEnumerator3)
    )
    Particles.ColorRedSmall:LinearColor=&MakeLinearColor(
     R:NiagaraFloat=1
     G:NiagaraFloat=0.2
     B:NiagaraFloat=0.2
     A:NiagaraFloat=&MakeFloatFromLinearColor(LinearColor=$Particles.Color, Channel:ENiagaraFloatFromLinearColorOptions=::NewEnumerator3)
    )
    Particles.ScaleBlueSmall:Vector3f=&Multiply_VectorByFloat(Vector=$Particles.Scale, Float=&Subtract_Float(A=1, B=0.05))
    Particles.ScaleRedSmall:Vector3f=&Multiply_VectorByFloat(Vector=$Particles.Scale, Float=&Add_Float(A=1, B=0.05))
   ) {isSetParameters=true}
   /SetVariables(
    Particles.ColorBlueBig:LinearColor=&MakeLinearColor(
     R:NiagaraFloat=0.1
     G:NiagaraFloat=0.1
     B:NiagaraFloat=1
     A:NiagaraFloat=&MakeFloatFromLinearColor(LinearColor=$Particles.Color, Channel:ENiagaraFloatFromLinearColorOptions=::NewEnumerator3)
    )
    Particles.ColorRedBig:LinearColor=&MakeLinearColor(
     R:NiagaraFloat=1
     G:NiagaraFloat=0.1
     B:NiagaraFloat=0.1
     A:NiagaraFloat=&MakeFloatFromLinearColor(LinearColor=$Particles.Color, Channel:ENiagaraFloatFromLinearColorOptions=::NewEnumerator3)
    )
    Particles.ScaleBlueBig:Vector3f=&Multiply_VectorByFloat(Vector=$Particles.Scale, Float=&Subtract_Float(A=1, B=0.02))
    Particles.ScaleRedBig:Vector3f=&Multiply_VectorByFloat(Vector=$Particles.Scale, Float=&Add_Float(A=1, B=0.02))
   ) {isSetParameters=true}
   &OffsetPosition(Position Offset=&Multiply_VectorByFloat(
     Vector=&Multiply_VectorByFloat(
      Vector=&RandomRangeVector(Minimum=<-0.3 -0.3 0.7>, Maximum=<0.3 0.3 1>)
      Float=&MakeCustomFloatFromBool(Bool=&RandomBool(Evaluation Type=::NewEnumerator1), True Float=1, False Float=-1)
     )
     Float=1.5
    ))
   /SetVariables(Particles.PositionBlueSmall:NiagaraPosition=$Particles.Position, Particles.PositionRedSmall:NiagaraPosition=$Particles.Position) {isSetParameters=true}
   /SetVariables(Particles.PositionBlueBig:NiagaraPosition=$Particles.Position, Particles.PositionRedBig:NiagaraPosition=$Particles.Position) {isSetParameters=true}
   &OffsetPosition(
    Position Offset=&Multiply_VectorByFloat(
     Vector=&Multiply_VectorByFloat(
      Vector=&RandomRangeVector(Minimum=<-0.3 -0.3 0.7>, Maximum=<0.3 0.3 1>)
      Float=&MakeCustomFloatFromBool(Bool=&RandomBool(Evaluation Type=::NewEnumerator1), True Float=1, False Float=-1)
     )
     Float=1
    )
    Position to Offset:NiagaraPosition=$Particles.PositionRedSmall
   ) {module=OffsetPosition001}
   &OffsetPosition(
    Position Offset=&Multiply_VectorByFloat(
     Vector=&Multiply_VectorByFloat(
      Vector=&RandomRangeVector(Minimum=<-0.3 -0.3 0.7>, Maximum=<0.3 0.3 1>)
      Float=&MakeCustomFloatFromBool(Bool=&RandomBool(Evaluation Type=::NewEnumerator1), True Float=1, False Float=-1)
     )
     Float=1
    )
    Position to Offset:NiagaraPosition=$Particles.PositionBlueSmall
   ) {module=OffsetPosition002}
   &OffsetPosition(
    Position Offset=&Multiply_VectorByFloat(
     Vector=&Multiply_VectorByFloat(
      Vector=&RandomRangeVector(Minimum=<-0.3 -0.3 0.7>, Maximum=<0.3 0.3 1>)
      Float=&MakeCustomFloatFromBool(Bool=&RandomBool(Evaluation Type=::NewEnumerator1), True Float=1, False Float=-1)
     )
     Float=1
    )
    Position to Offset:NiagaraPosition=$Particles.PositionRedBig
   ) {module=OffsetPosition003}
   &OffsetPosition(
    Position Offset=&Multiply_VectorByFloat(
     Vector=&Multiply_VectorByFloat(
      Vector=&RandomRangeVector(Minimum=<-0.3 -0.3 0.7>, Maximum=<0.3 0.3 1>)
      Float=&MakeCustomFloatFromBool(Bool=&RandomBool(Evaluation Type=::NewEnumerator1), True Float=1, False Float=-1)
     )
     Float=1
    )
    Position to Offset:NiagaraPosition=$Particles.PositionBlueBig
   ) {module=OffsetPosition004}
   /SetVariables(
    Particles.DynamicMaterialParameter0_Big:Vector4f=&MakeVector4(
     X=&Multiply_Float(A=&MakeFloatFromVector(Vector=$Particles.Scale, Channel=::NewEnumerator0), B=7)
     Y=&FloatFromCurve(
      FloatCurve=DataInterface() {propertyValues {Curve {
         defaultValue=1
         keys [
          {arriveTangent=7.3402, interpMode=RCIM_Cubic, leaveTangent=7.3402, tangentMode=RCTM_User, time=0, value=0}
          <at 0.2931 0.8277 RCIM_Cubic>
          {arriveTangent=-2.5257, interpMode=RCIM_Cubic, leaveTangent=-2.5257, tangentMode=RCTM_User, time=1, value=0}
         ]
        }}}
      Scale Curve=&Multiply_Float(
       A=&RandomRangeFloat(Minimum=30, Maximum=60)
       B=&RemapRange(Input Value To Remap=$User.SpawnRateMult, Input Min=1, Input Max=2, Output Min=0.5, Output Max=2)
      )
     )
     Z=&RandomRangeFloat(Minimum=0, Maximum=1)
     W=0
    )
    Particles.DynamicMaterialParameter1_Big:Vector4f=&MakeVector4(
     X=&MakeFloatFromVector4(Vector 4=$Particles.DynamicMaterialParameter1_Big, Channel=::NewEnumerator0)
     Y=&MakeFloatFromVector4(Vector 4=$Particles.DynamicMaterialParameter1_Big, Channel=::NewEnumerator1)
     Z=0.05
     W=1
    )
    Particles.DynamicMaterialParameter2_Big:Vector4f=&MakeVector4(X=2, Y=0, Z=0, W=0)
   ) {isSetParameters=true}
   /SetVariables(Particles.DynamicMaterialParameter0_Refraction:Vector4f=&MakeVector4(
     X=&FloatFromCurve(
      FloatCurve=DataInterface() {propertyValues {Curve {keys [{arriveTangent=-0.003, interpMode=RCIM_Cubic, leaveTangent=-0.003, tangentMode=RCTM_User, time=0, value=1.002}, <at 1 1 RCIM_Cubic>]}}}
      Scale Curve=&Multiply_Float(
       A=&FN_FrontEndFloatModifier(Gameplay=1, Front End=1.005)
       B=&RemapRange(
        Input Value To Remap=&Abs_Float(Float:NiagaraFloat=&DotProduct(Vector A=&ReturnCameraVector(Camera Vectors=::NewEnumerator1, Coordinate Space=::Simulation), Vector B=<0 1 0>))
        Input Min=0.5
        Input Max=1
        Output Min=1.002
        Output Max=1.005
       )
      )
     )
     Y=1
     Z=1
     W=0
    )) {isSetParameters=true}
  ]
  rendererClasses [&NiagaraMeshRendererProperties]
  renderers [
   {rendererClass=&NiagaraMeshRendererProperties, rendererIndex=0}
   {rendererClass=&NiagaraMeshRendererProperties, rendererIndex=1}
   {rendererClass=&NiagaraMeshRendererProperties, rendererIndex=2}
   {rendererClass=&NiagaraMeshRendererProperties, rendererIndex=3}
   {rendererClass=&NiagaraMeshRendererProperties, rendererIndex=4}
   {rendererClass=&NiagaraMeshRendererProperties, rendererIndex=5}
   {rendererClass=&NiagaraMeshRendererProperties, rendererIndex=6}
  ]
 }
 Speaker_Ring_Streaks_Glows {
  emitterUpdateScript [
   &EmitterState(
    Inactive Response=::NewEnumerator1
    Life Cycle Mode=::NewEnumerator1
    Loop Behavior=::NewEnumerator1
    Loop Delay=0.5
    Loop Duration Mode=::NewEnumerator1
    UseLoopDelay=true
   )
   &SpawnBurst_Instantaneous(Spawn Count=2, Use Spawn Probability=false, UseLoopCountLimit=false)
   &SpawnBurst_Instantaneous(Spawn Count=2, Spawn Group=1, Use Spawn Probability=false, UseLoopCountLimit=false) {module=SpawnBurst_Instantaneous002}
   &SpawnBurst_Instantaneous(Spawn Count=2, Spawn Group=2, Use Spawn Probability=false, UseLoopCountLimit=false) {module=SpawnBurst_Instantaneous001}
   &SpawnBurst_Instantaneous(Spawn Count=2, Spawn Group=3, Use Spawn Probability=false) {module=SpawnBurst_Instantaneous003}
  ]
  particleSpawnScript [
   &PartitionParticles()
   /SetVariables(Particles.VisibilityTag=&MergeHelper_IntSwitch(0=0, 1=1, 2=2, 3=3, 4=?, 5=?, Index=&SpawnGroup(), Num=::NewEnumerator2)) {isSetParameters=true}
   &InitializeParticle(
    Color=&MergeHelper_ColorSwitch(
     0=<rgba 9 8.6191 6.7148 0.06>
     1=<rgba 5 5 5 1>
     2=<rgba 5 5 5 1>
     3=<rgba 1.5 1.5 1.5 0.105>
     Index=$Particles.VisibilityTag
     Num=::NewEnumerator2
    )
    Lifetime=&MergeHelper_FloatSwitch(0=1, 1=5, 2=5, 3=0.5, Index=$Particles.VisibilityTag, Num=::NewEnumerator2)
    Mesh Scale=&MergeHelper_VectorSwitch(
     0=&VectorFromFloat(Value=0.05)
     1=<0 0 0>
     2=<0 0 0>
     3=<4.5 4.5 0.5>
     Index=$Particles.VisibilityTag
     Num=::NewEnumerator2
    )
    Mesh Scale Mode=::NewEnumerator4
    Sprite Size Mode=::NewEnumerator3
    Uniform Sprite Size=400
    UsePositionOffset=false
    Write Material Random=true
   )
   &InitialMeshOrientation(
    Mesh Orientation Mode=::NewEnumerator6
    Orientation Quaternion:Quat4f=&MergeHelper_QuatSwitch(
     0:Quat4f=<1 0 0 1>
     1:Quat4f=<0 0 0 0>
     2:Quat4f=<0 0 0 0>
     3:Quat4f=<-1 0 0 1>
     4:Quat4f=?
     5:Quat4f=?
     6:Quat4f=?
     7:Quat4f=?
     8:Quat4f=?
     9:Quat4f=?
     Index:NiagaraInt32=$Particles.VisibilityTag
     Num:MergeHelper_SwitcherEnum=::NewEnumerator2
    )
    Rotation=&MergeHelper_VectorSwitch(
     0=<0 0 0>
     1=<0 0 0>
     2=<0 0 0>
     3=&RandomRangeVector(Minimum=<0 0 0>, Maximum=<0 0 1>)
     Index=$Particles.VisibilityTag
     Num=::NewEnumerator2
    )
    Use Rotation Vector=true
   )
  ]
  particleUpdateScript [
   &ParticleState(Kill Particles When Lifetime Has Elapsed=false, Let Infinitely Lived Particles Die When Emitter Deactivates=true)
   &ScaleSpriteSize(Uniform Curve Sprite Scale=DataInterface() {propertyValues {Curve {
       defaultValue=0
       keys [
        {arriveTangent=9.0725, interpMode=RCIM_Cubic, leaveTangent=9.0725, tangentMode=RCTM_User, time=0, value=0}
        {arriveTangent=-2.5249, interpMode=RCIM_Cubic, leaveTangent=-2.5249, tangentMode=RCTM_User, time=0.0984, value=0.975}
       ]
      }}})
   &SkeletalMeshLocation(
    Bone / Socket Index=$Particles.PartitionParticles.Partition
    Bone Sampling Mode=::NewEnumerator4
    MaskSpawnGroup=false
    Sampled Position Offset=&MergeHelper_VectorSwitch(0=<-2 5 -1>, 1=<0 7.5 0>, 2=<0 7.5 0>, 3=<0 0 0>, Index=$Particles.VisibilityTag, Num=::NewEnumerator2)
    Skeletal Mesh=$System.Skeletal Mesh Reader_Speakers
   )
   &DynamicMaterialParameters(
    Index 0 Param 1=&Multiply_Float(A=&MakeFloatFromVector(Vector=$Particles.Scale, Channel=::NewEnumerator0), B=1.2)
    Index 0 Param 2=50
    Index 0 Param 3=0.25
    Index 0 Param 4=50
    Index 1 Param 1=15
    Index 1 Param 2=-0.07
    Index 1 Param 3=0.1
    Index 1 Param 4=$Particles.Lifetime
    Index 2 Param 1=2
    Module.Index0Param0Write=true
    Module.Index0Param1Write=true
    Module.Index0Param2Write=true
    Module.Index0Param3Write=true
    Module.Index1Param0Write=true
    Module.Index1Param1Write=true
    Module.Index1Param2Write=true
    Module.Index1Param3Write=true
    Module.Index2Param0Write=true
    Module.Index2Param1Write=true
    Module.Index2Param2Write=true
    Module.Index2Param3Write=true
    Module.Index3Param0Write=true
    Module.Index3Param1Write=true
    Module.Index3Param2Write=true
    Module.Index3Param3Write=true
    Write Parameter Index 1=true
    Write Parameter Index 2=true
    Write Parameter Index 3=true
   )
   /SetVariables(Particles.DynamicMaterialParameter0_Glow:Vector4f=<3 1 0 1>) {isSetParameters=true}
   /SetVariables(Particles.DynamicMaterialParameter0_Streaks:Vector4f=<0 1 1 1>) {isSetParameters=true}
   &ScaleMeshSize(Scale Factor=&MergeHelper_VectorSwitch(
     0=&VectorFromFloat(Value=&Multiply_Float(
       A=&RemapRange(Input Value To Remap=$User.SpawnRateMult, Input Min=1, Input Max=2, Output Min=1, Output Max=1.1)
       B=&Lerp_Float(
        A=&Waveform(Global Input Position=-0.17, "Waveform [1]"=::NewEnumerator0, "[1] Frequency"=1, "[1] Amplitude Min/Max"=<1 1.2>)
        B=&Waveform(Global Input Position=-0.17, "Waveform [1]"=::NewEnumerator0, "[1] Phase Scale"=4, "[1] Frequency"=1, "[1] Amplitude Min/Max"=<1.8 2.2>)
        Alpha=&FloatFromCurve(FloatCurve=DataInterface() {propertyValues {Curve {keys [{interpMode=RCIM_Cubic, leaveTangent=1.8912, tangentMode=RCTM_Break, time=0, value=0}, <at 1 1>]}}})
       )
      ))
     1=<0 0 0>
     2=<0 0 0>
     3=&VectorFromFloat(Value=&FloatFromCurve(
       FloatCurve=DataInterface() {propertyValues {Curve {
          defaultValue=0
          keys [
           {arriveTangent=9.0725, interpMode=RCIM_Cubic, leaveTangent=9.0725, tangentMode=RCTM_User, time=0, value=0}
           {arriveTangent=-2.5249, interpMode=RCIM_Cubic, leaveTangent=-2.5249, tangentMode=RCTM_User, time=0.0984, value=0.975}
          ]
         }}}
       Scale Curve=&Waveform("Waveform [1]"=::NewEnumerator0, "[1] Phase Scale"=4, "[1] Frequency"=1, "[1] Amplitude Min/Max"=<0.75 1>)
      ))
     Index=$Particles.VisibilityTag
     Num=::NewEnumerator2
    ))
   /SetVariables(Particles.MeshOrientation=&MakeQuatFromAxisAngle(
     Axis=&MergeHelper_VectorSwitch(0=<1 0 0>, 1=<0 0 0>, 2=<0 0 0>, 3=<-1 0 0>, Index=$Particles.VisibilityTag, Num=::NewEnumerator2)
     AngleInDegrees=90
    )) {isSetParameters=true}
   &MeshRotationForce(Rotation=&MakeVector(
     X=&MergeHelper_FloatSwitch(
      0=0
      1=0
      2=0
      3=&FloatFromCurve(FloatCurve=DataInterface() {propertyValues {Curve {keys [<at 0 1>,<at 1 0>]}}})
      Index=$Particles.VisibilityTag
      Num=::NewEnumerator2
     )
     Y=0
     Z=0
    ))
   &ScaleColor(
    Scale Alpha=&Multiply_Float(
     A=&FNTimeOfDayScalar(Night Scale=1, Morning Scale=1, Evening Scale=1)
     B=&MergeHelper_FloatSwitch(
      0=&Multiply_Float(
       A=&FNTimeOfDayScalar(Night Scale=0.5, Morning Scale=0.8, Evening Scale=0.8)
       B=&RemapRange(Input Value To Remap=$User.Burst, Input Min=0, Input Max=1, Output Min=0.2, Output Max=0.5)
      )
      1=&Multiply_Float(
       A=&FNTimeOfDayScalar(Night Scale=0.0125, Morning Scale=0.0125, Day Scale=0.011, Evening Scale=0.0125)
       B=&RemapRange(Input Value To Remap=$User.Burst, Input Min=0, Input Max=1, Output Min=0.8, Output Max=1)
      )
      2=1
      3=&Multiply_Float(
       A=&FN_FrontEndFloatModifier(Gameplay=1, Front End=1.25)
       B=&RemapRange(Input Value To Remap=$User.Burst, Input Min=0, Input Max=1, Output Min=0.2, Output Max=1)
      )
      Index=$Particles.VisibilityTag
      Num=::NewEnumerator2
     )
    )
    Scale RGB=&VectorFromFloat(Value=&FNTimeOfDayScalar(Night Scale=1, Morning Scale=1, Evening Scale=1))
    ScaleA=true
    ScaleRGB=true
   ) {module=ScaleColor001}
   &SolveForcesAndVelocity(Clamp Velocity=false, Limit Acceleration=false)
   &FN_Light_Attributes(
    Color=<rgba 55 53.5014 46.0087 1>
    Create Color Attribute=true
    Create Exponent Attribute=true
    Create Position Attribute=true
    Create Radius Attribute=true
    Enable Particle Light Enable=true
    Exponent=1
    Light Enabled=true
    Position=&AddVectorToPosition(Position=$Particles.Position, Vector=<0 5 0>)
    Radius=150
    Scale Color=&Multiply_Float(
     A=&FNTimeOfDayScalar(Night Scale=0.4, Morning Scale=0.6, Evening Scale=0.3)
     B=&Multiply_Float(
      A=&FloatFromCurve(
       FloatCurve=DataInterface() {propertyValues {Curve {
          defaultValue=0
          keys [
           {arriveTangent=9.0725, interpMode=RCIM_Cubic, leaveTangent=9.0725, tangentMode=RCTM_User, time=0, value=0}
           {arriveTangent=-2.5249, interpMode=RCIM_Cubic, leaveTangent=-2.5249, tangentMode=RCTM_User, time=0.0984, value=0.975}
          ]
         }}}
       Scale Curve=4
      )
      B=&RemapRange(Input Value To Remap=$User.Burst, Input Min=0, Input Max=1, Output Min=0.4, Output Max=1)
     )
    )
   )
  ]
  rendererClasses [
   &NiagaraMeshRendererProperties
   &NiagaraSpriteRendererProperties
   &NiagaraLightRendererProperties
  ]
  renderers [
   {rendererClass=&NiagaraMeshRendererProperties, rendererIndex=0}
   {rendererClass=&NiagaraSpriteRendererProperties, rendererIndex=1}
   {rendererClass=&NiagaraLightRendererProperties, rendererIndex=2}
   {rendererClass=&NiagaraMeshRendererProperties, rendererIndex=3}
  ]
 }
}
summary {
 systemName=NS_Emote_GothClaws_Soundwaves
 emitters [
  {
   emitterName=Speaker_Ring_Streaks_Glows
   rendererClasses [
    &NiagaraMeshRendererProperties
    &NiagaraSpriteRendererProperties
    &NiagaraLightRendererProperties
   ]
  }
  {emitterName=Rings, rendererClasses [&NiagaraMeshRendererProperties]}
  {
   emitterName=Blinking_Streaks
   rendererClasses [&NiagaraMeshRendererProperties, &NiagaraSpriteRendererProperties]
  }
 ]
 userVariables [
  User.bIsFrontend {type=NiagaraBool}
  User.bIsFrontendPreview {type=NiagaraBool}
  User.Burst {type=NiagaraFloat, defaultValue=1, description=""}
  User.Chromatic_Offset {type=NiagaraFloat, defaultValue=0.02, description=""}
  User.ColorIntensity {type=NiagaraFloat, defaultValue=2, description=""}
  User.FEOffset {type=NiagaraFloat, defaultValue=10, description=""}
  User.FEScale {type=NiagaraFloat, defaultValue=0.6, description=""}
  User.Intensity {type=NiagaraFloat, defaultValue=1, description=""}
  User.IsSpawnDust {type=NiagaraBool, defaultValue=true}
  User.SpawnRateMult {type=NiagaraFloat, defaultValue=1, description=""}
 ]
}
